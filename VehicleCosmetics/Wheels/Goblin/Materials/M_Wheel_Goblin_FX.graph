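# Translucent glowing effect material for the skeletal Goblin vehicle wheel, with emissive set by CustomColor times EmissiveMultiply.
# Two caustics samples pan at different rates, scaled by PanSpeedAdjust, inside a static mask built from a shockwave curve texture and a stepped V gradient.
# PaintedBlack_Opacity blends that animated opacity toward a faint copy of the mask alone.
# The material attributes pass through MF_Vehicle_FNPostFX before output.
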Material /VehicleCosmetics/Wheels/Goblin/Materials/M_Wheel_Goblin_FX
graph {
 properties {
  bAutomaticallySetUsageInEditor=false
  bUseMaterialAttributes=true
  bUsedWithMeshDeformer=true
  bUsedWithSkeletalMesh=true
  blendMode=BLEND_Translucent
  pixelDepthOffsetMode=PDOM_Legacy
  previewMesh=/VehicleCosmetics/Wheels/Goblin/Meshes/SK_Wheel_Goblin
  usedShadingModels=Default Lit
 }
 nodes [
  n1 = VectorParameter() {parameterName=CustomColor, defaultValue=<rgba 0.1356 1 0 1>}
  n2 = ScalarParameter() {defaultValue=32, parameterName=EmissiveMultiply}
  n3 = TextureCoordinate() {uTiling=0.4, vTiling=0.1}
  n4 = Multiply(A=Time(), B=ScalarParameter() {defaultValue=1, parameterName=PanSpeedAdjust})
  n5 = TextureSample(UVs=Panner(Coordinate=@n3, Time=@n4) {speedX=0.125, speedY=1}) {texture=/VehicleCosmetics/Wheels/Goblin/Textures/Caustics_D}
  n6 = TextureSample(UVs=Panner(Coordinate=@n3, Time=@n4) {speedX=0.075, speedY=0.067}) {texture=&Caustics_D}
  n7 = TextureCoordinate() {uTiling=2, vTiling=2.5}
  n8 = Divide(A=ComponentMask(None=@n7) {g=true}) {constB=2.5}
  n9 = Constant2Vector() {g=-1.55}
  n10 = Multiply(
   A=Step(Y=OneMinus(None=@n8)) {constX=0.38}
   B=TextureSample(UVs=Add(A=@n7, B=@n9)) {texture=/VehicleCosmetics/Wheels/Goblin/Textures/Curve_Shockwaves_Pack}.R
  )
  n11 = Clamp(None=ConstantBiasScale(None=@n10) {bias=-0.05, scale=2})
  n12 = Add(A=Add(A=@n5.R, B=Add(A=@n6.G, B=@n5.R)), B=@n11)
  n13 = Multiply(A=@n10, B=Constant() {r=0.1})
  n14 = Add(A=Multiply(A=Clamp(None=@n12), B=@n11), B=@n13)
  n15 = ScalarParameter() {defaultValue=0.5, parameterName=PaintedBlack_Opacity}
  MaterialFunctionCall!(In=MakeMaterialAttributes!(
    BaseColor
    Metallic
    Specular
    Roughness
    Anisotropy
    EmissiveColor=Multiply(A=@n1.RGB, B=@n2)
    Opacity=LinearInterpolate(A=@n14, B=@n13, Alpha=@n15) {constAlpha=0}
    OpacityMask
    Normal
    Tangent
    WorldPositionOffset
    SubsurfaceColor
    ClearCoat
    ClearCoatRoughness
    AmbientOcclusion
    Refraction
    CustomizedUVs_0
    CustomizedUVs_1
    CustomizedUVs_2
    CustomizedUVs_3
    CustomizedUVs_4
    CustomizedUVs_5
    CustomizedUVs_6
    CustomizedUVs_7
    PixelDepthOffset
    ShadingModel
    Displacement
   )) {materialFunction=/VehicleCosmetics/Materials/MaterialFunctions/MF_Vehicle_FNPostFX}
 ]
}
